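# Translucent UI material that procedurally draws reticle prongs for HoneyFiber Incident, with a constant warm orange base color.
# UVs are offset, skewed, mirrored through a 2x1 symmetry function and rotated, then pass through the shared reticle UV function.
# An isosceles trapezoid SDF shapes each prong, and anti-aliasing at two thresholds yields the fill and shadow outline masks.
# Scalar parameters expose prong position, size, skew, rotation and spread, plus outline and reticle scale.
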Material /HoneyFiber_IncidentUI/Art/Materials/M_HoneyFiber_Incident_Reticle
graph {
 properties {bAutomaticallySetUsageInEditor=false, blendMode=BLEND_Translucent, materialDomain=MD_UI, usedShadingModels=Default Lit}
 nodes [
  Multiply(A=NamedRerouteUsage(), B=NamedRerouteUsage())
  n1 = Add(
   A=NamedRerouteUsage()
   B=AppendVector(
    A=ScalarParameter() {defaultValue=-0.024, group=Prongs, parameterName=Position X}
    B=ScalarParameter() {defaultValue=-0.28, group=Prongs, parameterName=PositionY}
   )
  )
  n2 = LinearInterpolate(A=ScalarParameter() {defaultValue=-4, group=Prongs, parameterName=SkewAngle, sortPriority=6}, B=Constant()) {constA=1, constB=0.5}
  n3 = MaterialFunctionCall(
   UVs=MaterialFunctionCall!(
    UV=MaterialFunctionCall!(UV=@n1, Skew Angle=@n2) {materialFunction=/FortUILibrary/MaterialFunctions/Basics/MF_UI_Skew}.Skewed UVs
    Mirror Y
    Swizzle UV
   ) {materialFunction=/FortUILibrary/MaterialFunctions/Basics/MF_UI_Symmetry_2x1_New}.Symmetry
   "Rotation Angle (0-1)"=ScalarParameter() {defaultValue=0.1, group=Prongs, parameterName=Rotation, sortPriority=5}
  ) {materialFunction=&CustomRotator}
  n4 = Constant() {r=1}
  n5 = NamedRerouteUsage()
  n6 = ScalarParameter() {defaultValue=-0.9863, group=Prongs, parameterName=Spread, sortPriority=4}
  n7 = Multiply(A=MaterialFunctionCall!(
    UVs=MaterialFunctionCall!(
     Reticle UVs=@n3.Rotated Values
     Reticle Sides=@n4
     Reticle Scale=@n5
     Reticle Rotation
     Prong Scale
     Prong Rotation
     Spread Offset
     Positional Offset=@n6
     "Spread Input Range [Min,Max]"
     "Spread Remapped Range [Min,Max]"
    ) {materialFunction=/FortUILibrary/Systems/Reticles/MaterialFunctions/MF_UI_Reticle_UVs}.UVs
    TopWidth=ScalarParameter() {defaultValue=0.062, group=Prongs, parameterName=TopWidth, sortPriority=2}
    BottomWidth=ScalarParameter() {defaultValue=0.15, group=Prongs, parameterName=BottomWidth, sortPriority=3}
    Height=ScalarParameter() {defaultValue=0.608, group=Prongs, parameterName=Height, sortPriority=1}
    OutlineThickness
    OutlineAlignment
    GlowMin
    GlowMax
   ) {materialFunction=/FortUILibrary/MaterialFunctions/Shapes/MF_UI_IsoscelesTrapezoid}.SDF)
  n8 = MaterialFunctionCall(DistanceField=@n7) {materialFunction=&MF_UI_AntiAliasing}
  NamedRerouteDeclaration(None=@n8.Inner) {name=HorizontalLinesColor, nodeColor=<rgba 0 0 0 1>}
  n9 = ScalarParameter() {defaultValue=0.04, parameterName=Outline}
  NamedRerouteDeclaration(None=LinearInterpolate(
    A=@n8.Inner
    B=MaterialFunctionCall(DistanceField=@n7, Threshold=@n9) {materialFunction=&MF_UI_AntiAliasing}.Inner
   )) {name=HorizontalLinesShadow, nodeColor=<rgba 0 0 0 1>}
  NamedRerouteDeclaration(None=TextureCoordinate()) {name=UV_1, nodeColor=<rgba 0 0 0 1>}
  NamedRerouteDeclaration(None=ScalarParameter() {defaultValue=1.35, parameterName=ReticleScale}) {name=Reticle Scale, nodeColor=<rgba 0 0 0 1>}
  NamedRerouteDeclaration(None=Constant3Vector() {constant=<rgba 1 0.6867 0.2831 1>}) {name=ColorBase, nodeColor=<rgba 0 0 0 1>}
  NamedRerouteUsage()
 ]
}
